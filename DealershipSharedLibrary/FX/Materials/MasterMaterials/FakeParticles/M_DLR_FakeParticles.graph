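# Translucent, unlit, two-sided master material for fake particle effects, enabled for Nanite and Niagara mesh particles.
# MF_FakeParticles does the core work, producing emissive, opacity and world position offset, with inputs for SubUV animation, random color, fake lighting, size, spin and orbit.
# Its opacity passes through a depth fade function with an exposed distance parameter, and its emissive through a shared emissive adjustment function.
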Material /DealershipSharedLibrary/FX/Materials/MasterMaterials/FakeParticles/M_DLR_FakeParticles
graph {
 properties {
  bUsedWithNanite=true
  bUsedWithNiagaraMeshParticles=true
  blendMode=BLEND_Translucent
  shadingModel=MSM_Unlit
  twoSided=true
  usedShadingModels=Unlit
 }
 nodes [
  n1 = MaterialFunctionCall!(
   ***General Parameters***
   UseRoundRect?
   CustomOpacity
   "UseTransitionAlpha?0=false1=true"
   Texture
   "Use RandomColor 0=False 1=True"
   "UseFakeLightning? 0=False 1=True"
   LightDirection
   SubUV_Texture
   CustomUvs
   SubImages
   CustomFrame
   SubUV_Speed
   ***WPO Parameters***
   "UseRandomSize? 0=False 1=True"
   Scale WPO
   "UseRandomKillSize? 0=False 1=True"
   RotateScaleX
   RotateScaleY
   "UseSpin? 0=False 1=True"
   SpinSpeed
   "Orbit 0=False X=Intensity"
   RotateScaleX_1
  ) {materialFunction=/DealershipSharedLibrary/FX/Materials/MaterialFunctions/MF_FakeParticles}
  MaterialFunctionCall(Opacity In=@n1.Opacity, DepthFade Distance=ScalarParameter() {defaultValue=100, group=DepthFade, parameterName=DepthFade Distance}) {materialFunction=&MF_VCVFX_DepthFade}
  MaterialFunctionCall(Emissive=@n1.Emissive) {materialFunction=&MF_FX_EmissiveAdjustments}
  Reroute(None=@n1.WorldPosOffset)
 ]
}
